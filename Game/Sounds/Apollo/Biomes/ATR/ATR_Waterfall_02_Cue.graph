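# Looping waterfall ambience cue for the ATR biome, routed to the ambient water sound class with waterfall attenuation.
# A quality-level root picks among four pitch-modulated branches: the first randomly chooses between two loops, the second plays one loop, and the last two use an OPT loop variant.
# A concurrency override caps simultaneous instances, stopping the quietest, and the debug flag is left on.
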SoundCue /Game/Sounds/Apollo/Biomes/ATR/ATR_Waterfall_02_Cue
root=QualityLevel_0
nodes [
 {id=Modulator_1, children [WavePlayer_8], properties {pitchMax=1.1, pitchMin=0.9}}
 {id=Modulator_3, children [WavePlayer_0], properties {pitchMax=1.1, pitchMin=0.9}}
 {id=Modulator_4, children [Random_1], properties {pitchMax=1.1, pitchMin=0.99}}
 {id=Modulator_5, children [WavePlayer_7], properties {pitchMax=1.2, pitchMin=0.99}}
 {id=QualityLevel_0, children [Modulator_4,Modulator_5,Modulator_1,Modulator_3]}
 {id=Random_1, children [WavePlayer_5,WavePlayer_6], properties {weights [1,1]}}
 {id=WavePlayer_0, properties {bLooping=true, soundWaveAssetPtr=&/ATR_Waterfall_Loop_01_OPT}}
 {id=WavePlayer_5, properties {bLooping=true, soundWaveAssetPtr=&/ATR_Waterfall_Loop_01}}
 {id=WavePlayer_6, properties {bLooping=true, soundWaveAssetPtr=&/ATR_Waterfall_Loop_02}}
 {id=WavePlayer_7, properties {bLooping=true, soundWaveAssetPtr=&ATR_Waterfall_Loop_01}}
 {id=WavePlayer_8, properties {bLooping=true, soundWaveAssetPtr=&ATR_Waterfall_Loop_01_OPT}}
]
properties {
 attenuationSettings=&/ATR_Waterfall_Att
 bDebug=true
 bOverrideConcurrency=true
 maxDistance=7000
 soundClassObject=/Game/Sounds/FortSoundClassesAndModes/Fort_Ambient_Water
 volumeMultiplier=1
 concurrencyOverrides {maxCount=4, resolutionRule=StopQuietest}
}
